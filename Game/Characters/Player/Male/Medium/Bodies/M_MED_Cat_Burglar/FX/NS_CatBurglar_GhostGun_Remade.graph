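# Niagara system for the Cat Burglar character's ghost gun, adding a purple spectral aura to a weapon mesh, previewed on the GhostMidas drum gun.
# Two sprite emitters spawn continuously: large faint airglow puffs and smaller SubUV-animated wisps that narrow and lengthen.
# Spawn group selects between the FX_Idle_Start and FX_Idle_End sockets.
# Particles resample their socket during update, drifting along a curve-eased, randomized, Z-biased offset while alpha fades in and out.
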NiagaraSystem /Game/Characters/Player/Male/Medium/Bodies/M_MED_Cat_Burglar/FX/NS_CatBurglar_GhostGun_Remade
dependencies {
 usedDataInterfaces [
  &NiagaraDataInterfaceSkeletalMesh
  &NiagaraDataInterfaceCurve
  &NiagaraDataInterfaceVector2DCurve
 ]
 usedDynamicInputs [
  &SimulationPosition
  &UniformRangedVector2D
  &Vector2DFromFloat
 ]
 usedModules [&SystemState]
}
emitters {
 NE_Airglow001 {
  emitterUpdateScript [
   &EmitterState(Life Cycle Mode=::NewEnumerator1, Loop Duration=1, UseLoopDelay=false)
   &SpawnRate(SpawnRate=4, Use Spawn Probability=false)
  ]
  particleSpawnScript [
   &InitializeParticle(
    Color=<rgba 0.0708 0.0242 0.3 0.01>
    Lifetime=2
    Sprite Size Mode=::NewEnumerator3
    Uniform Sprite Size=80
    UsePositionOffset=false
    Write Material Random=true
   ) {module=InitializeParticle001}
   &CameraOffset(Camera Offset Amount=&RandomRangeFloat(Minimum=-5, Maximum=20))
   &BoxLocation(Box Offset=<5 0 0>, Box Size=<30 10 10>, MaskSpawnGroup=false, Module.Override Seed=false, Rotate Box=false, Surface Only=false)
   &FN_SetSpawnGroup(Number of Spawn Groups=2)
  ]
  particleUpdateScript [
   &ParticleState()
   &SolveForcesAndVelocity(Clamp Velocity=false, Limit Acceleration=false)
   &ScaleColor(
    Scale Alpha=&FloatFromCurve(FloatCurve=DataInterface() {propertyValues {Curve {keys [<at -0.0057 0>,<at 0.5002 1.0023>,<at 1 0>]}}})
    ScaleA=true
    ScaleRGB=true
   )
   &SkeletalMeshLocation(
    Bone / Socket Index=$Particles.SpawnGroup
    Inherit Velocity=false
    MaskSpawnGroup=false
    Mesh Sampling Type=::NewEnumerator5
    Sampled Position Offset=&Lerp_Vector(
     A=Linked() {linkedVariable {name=Particles.Initial.Position, type=NiagaraPosition}}
     B=&Add_Vector(
      A=Linked() {linkedVariable {name=Particles.Initial.Position, type=NiagaraPosition}}
      B=&RandomRangeVector(Minimum=<-3 -3 10>, Maximum=<3 3 5>)
     )
     Alpha=&FloatFromCurve(FloatCurve=DataInterface() {propertyValues {Curve {keys [
          {arriveTangent=-0.0008, interpMode=RCIM_Cubic, leaveTangent=-0.0008, tangentMode=RCTM_User, time=0.0133, value=0.043}
          {arriveTangent=1.3061, interpMode=RCIM_Cubic, leaveTangent=1.3061, tangentMode=RCTM_User, time=1, value=1}
         ]}}})
    )
    Sampled Position Offset Coordinate Space:ENiagaraMeshTransforms=::NewEnumerator3
    Skeletal Mesh=DataInterface() {propertyValues {
      DefaultMesh=null
      ExcludeBoneName=root
      OptionalComponentClass=null
      PreviewMesh=/Game/Weapons/FORT_Rifles/Mesh/SK_DrumGun_GhostMidas
      SkinningMode=SkinOnTheFly
      SoftSourceActor=null
      SourceMode=Default
      UvSetIndex=0
      WholeMeshLOD=-1
      bExcludeBone=false
      bReadDeformedGeometry=true
      bRequireCurrentFrameData=true
      ComponentTags []
      FilteredBones []
      FilteredSockets [FX_Idle_Start, FX_Idle_End]
      MeshUserParameter {parameter {name=None, typeDefHandle {registeredTypeIndex=1}, varData []}}
      SamplingRegions []
     }}
    Socket Sampling Mode=::NewEnumerator5
   )
  ]
  rendererClasses [&NiagaraSpriteRendererProperties]
  renderers [{rendererClass=&NiagaraSpriteRendererProperties}]
 }
 NE_Empty {
  emitterUpdateScript [
   &EmitterState(Life Cycle Mode=::NewEnumerator1, Loop Duration=1, UseLoopDelay=false)
   &SpawnRate(SpawnRate=10, Use Spawn Probability=false)
  ]
  particleSpawnScript [
   &InitializeParticle(
    Color=<rgba 0.4537 0.147 1 0.2>
    Lifetime=&RandomRangeFloat(Minimum=1.15, Maximum=1.6)
    Sprite Rotation Angle Max=50
    Sprite Rotation Angle Min=-50
    Sprite Rotation Mode=::NewEnumerator2
    Sprite Size Mode=::NewEnumerator1
    Uniform Sprite Size Max=23
    Uniform Sprite Size Min=16
    UsePositionOffset=false
    Write Material Random=true
   ) {module=InitializeParticle001}
   &SubUVAnimation(End Frame=6, SubUV Animation Mode:ENiagara_SubUVLookupMode=::NewEnumerator2)
   &BoxLocation(Box Offset=<5 0 0>, Box Size=<30 10 10>, MaskSpawnGroup=false, Module.Override Seed=false, Rotate Box=false, Surface Only=false)
   &FN_SetSpawnGroup(Number of Spawn Groups=2)
   &CameraOffset(Camera Offset Amount=&RandomRangeFloat(Minimum=-5, Maximum=20))
  ]
  particleUpdateScript [
   &ParticleState()
   &SkeletalMeshLocation(
    Bone / Socket Index=$Particles.SpawnGroup
    Inherit Velocity=false
    MaskSpawnGroup=false
    Mesh Sampling Type=::NewEnumerator5
    Sampled Position Offset=&Lerp_Vector(
     A=Linked() {linkedVariable {name=Particles.Initial.Position, type=NiagaraPosition}}
     B=&Add_Vector(
      A=Linked() {linkedVariable {name=Particles.Initial.Position, type=NiagaraPosition}}
      B=&RandomRangeVector(Minimum=<-3 -3 10>, Maximum=<3 3 5>)
     )
     Alpha=&FloatFromCurve(FloatCurve=DataInterface() {propertyValues {Curve {keys [
          {arriveTangent=-0.0008, interpMode=RCIM_Cubic, leaveTangent=-0.0008, tangentMode=RCTM_User, time=0.0133, value=0.043}
          {arriveTangent=1.3061, interpMode=RCIM_Cubic, leaveTangent=1.3061, tangentMode=RCTM_User, time=1, value=1}
         ]}}})
    )
    Sampled Position Offset Coordinate Space:ENiagaraMeshTransforms=::NewEnumerator3
    Skeletal Mesh=DataInterface() {propertyValues {
      DefaultMesh=null
      ExcludeBoneName=root
      OptionalComponentClass=null
      PreviewMesh=&SK_DrumGun_GhostMidas
      SkinningMode=SkinOnTheFly
      SoftSourceActor=null
      SourceMode=Default
      UvSetIndex=0
      WholeMeshLOD=-1
      bExcludeBone=false
      bReadDeformedGeometry=true
      bRequireCurrentFrameData=true
      ComponentTags []
      FilteredBones []
      FilteredSockets [FX_Idle_Start, FX_Idle_End]
      MeshUserParameter {parameter {name=None, typeDefHandle {registeredTypeIndex=1}, varData []}}
      SamplingRegions []
     }}
    Socket Sampling Mode=::NewEnumerator5
   )
   &ScaleSpriteSize(Scale Factor=&Vector2DFromCurve(Vector2Curve=DataInterface() {propertyValues {XCurve {keys [<at 0 1>,<at 1 0.5>]}, YCurve {keys [<at 0 0.8>,<at 1 1.4>]}}}, ScaleCurve=<1 1>))
   &ScaleColor(
    Scale Alpha=&FloatFromCurve(FloatCurve=DataInterface() {propertyValues {Curve {keys [<at 0 0>,<at 0.3747 0.7273>,<at 1 0>]}}})
    ScaleA=true
    ScaleRGB=true
   )
  ]
  rendererClasses [&NiagaraSpriteRendererProperties]
  renderers [{rendererClass=&NiagaraSpriteRendererProperties}]
 }
}
summary {
 systemName=NS_CatBurglar_GhostGun_Remade
 emitters [
  {emitterName=NE_Empty, rendererClasses [&NiagaraSpriteRendererProperties]}
  {emitterName=NE_Airglow001, rendererClasses [&NiagaraSpriteRendererProperties]}
 ]
 userVariables [User.WeaponMesh {
   type=Object
   description=""
   defaultValue {struct=NiagaraExt_VariableValue_Object, value {object=null, objectClass=&Object}}
  }]
}
